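# Parent material for the CanvasPrint Poster face accessory skin, adding kill-reactive and audio-reactive emissive glow between the shared character PreFX and PostFX functions.
# A Blueprint-set start time drives a fade-in/out pulse that blends in a reactive texture glow.
# Frontend-loop audio amplitude, with a cosine fake fallback input, combines with that pulse to blend in a tinted, boosted glow.
# This boosted glow is masked to dark base-color areas, with roughness-controlled shine-through.
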Material /BRCosmetics/Characters/Player/Male/Medium/Bodies/M_MED_CanvasPrint/Skins/CanvasPrint_Poster/Materials/M_CanvasPrint_Poster_FaceAcc_PARENT
graph {
 properties {
  bHasPixelAnimation=true
  bUseMaterialAttributes=true
  bUsedWithMeshDeformer=true
  bUsedWithSkeletalMesh=true
  pixelDepthOffsetMode=PDOM_Legacy
  shadingModel=MSM_FromMaterialExpression
  usedShadingModels=Default Lit | Subsurface | Clear Coat | Subsurface Profile | Cloth
 }
 nodes [
  n1 = ScalarParameter() {defaultValue=2.5, group=CanvasPrint_Reactivity, parameterName=AudioGlow_Intensity}
  n2 = MaterialFunctionCall() {materialFunction=&MF_CharacterParent_2023_PreFX}
  n3 = GetMaterialAttributes(None=@n2.Result) {attributeGetTypes [#1, #2, #3]}
  n4 = TextureSample() {
   samplerSource=SSM_Wrap_WorldGroupSettings
   texture=/BRCosmetics/Characters/Player/Male/Medium/Bodies/M_MED_CanvasPrint/Skins/CanvasPrint_Poster/Textures/T_CanvasPrint_Poster_FaceAcc_Reactive
  }
  n5 = ScalarParameter() {group=CanvasPrint_Reactivity, parameterName=ReactiveGlow_Desaturation}
  n6 = NamedRerouteUsage()
  n7 = LinearInterpolate(
   A=@n3.Emissive Color
   B=Multiply(
    A=Desaturation(None=@n4.RGB, Fraction=@n5)
    B=ScalarParameter() {defaultValue=1.5, group=CanvasPrint_Reactivity, parameterName=ReactiveGlow_Intensity}
   ) {constB=1.5}
   Alpha=Multiply(A=ScalarParameter() {defaultValue=1, group=CanvasPrint_Reactivity, parameterName=EnableKillReactivity, sortPriority=-10}, B=@n6)
  )
  n8 = Power(Base=@n4.RGB, Exp=ScalarParameter() {defaultValue=3, group=CanvasPrint_Reactivity, parameterName=AudioGlow_Exp}) {constExponent=3}
  n9 = Multiply(
   A=VectorParameter() {group=CanvasPrint_Reactivity, parameterName=AudioGlow_Color, defaultValue=<rgba 1.5117 0.3 3 1>}.RGB
   B=ScalarParameter() {defaultValue=2, group=CanvasPrint_Reactivity, parameterName=AudioGlow_Brightness}
  )
  n10 = ScalarParameter() {defaultValue=0.25, group=CanvasPrint_Reactivity, parameterName=Reactive_ShineThroughStrength, sliderMax=1}
  n11 = ScalarParameter() {defaultValue=0.5, group=CanvasPrint_Reactivity, parameterName=ReactiveMask_RoughnessTreshold}
  n12 = Saturate(None=Multiply(A=Subtract(A=@n3.Roughness, B=@n11), B=ScalarParameter() {defaultValue=5, parameterName=ReactiveMask_Contrast}) {constB=20})
  n13 = Desaturation(None=@n3.Base Color)
  n14 = ScalarParameter() {defaultValue=0.015, group=CanvasPrint_Reactivity, parameterName=ReactiveMask_BaseColorTreshold}
  n15 = Multiply(A=Multiply(A=Subtract(A=@n13, B=@n14)) {constB=-1}) {constB=100}
  n16 = Multiply(
   A=Add(A=Multiply(A=@n7, B=@n1), B=Multiply(A=@n8, B=@n9))
   B=Multiply(A=LinearInterpolate(A=@n10, Alpha=@n12) {constA=0.1}, B=Saturate(None=@n15))
  )
  n17 = Time()
  n18 = ScalarParameter() {defaultValue=1.3, group=Audio Reactivity, parameterName=FrontendFakeAudio_LoopLength, sortPriority=200}
  n19 = ConstantBiasScale(None=Cosine(None=Divide(A=@n17, B=@n18)))
  n20 = ScalarParameter() {defaultValue=0.5, group=Audio Reactivity, parameterName=FrontendFakeAudio_Amplitude, sortPriority=200}
  n21 = Power(Base=MaterialFunctionCall!(InAmplitude, InSpectrum, CustomFakeSpectrum, CustomFakeAmplitude=Multiply(A=@n19, B=@n20)) {materialFunction=/BRCosmetics/Global/MaterialFunctions/AudioReactivity/MF_AudioReactivity_FrontendLoop}.Amplitude) {constExponent=6}
  n22 = Multiply(
   A=Max(A=LinearInterpolate(A=@n6, B=@n21), B=@n21)
   B=ScalarParameter() {defaultValue=1, group=CanvasPrint_Reactivity, parameterName=EnableAudioReactivity, sortPriority=-10}
  )
  MaterialFunctionCall(
   Material=SetMaterialAttributes(MaterialAttributes=@n3.MaterialAttributes, Emissive Color=LinearInterpolate(A=@n7, B=@n16, Alpha=@n22)) {attributeSetTypes [#2]}
   M=@n2.M
   Pre Skinned Local Position=@n2.M
   Distance Blend=@n2.M
   Fresnel=@n2.M
  ) {materialFunction=&MF_CharacterParent_2023_PostFX}
  n23 = Divide(
   A=Subtract(A=Time(), B=ScalarParameter() {defaultValue=999999995904, group=CanvasPrint_Reactivity, parameterName=BP_GetTime, sortPriority=0})
   B=Max(A=ScalarParameter() {defaultValue=3, group=CanvasPrint_Reactivity, parameterName=ReactiveEyes_Duration, sortPriority=40}) {constB=0.0001}
  )
  n24 = Add(A=Multiply(A=Saturate(None=@n23)) {constB=2}) {constB=-1}
  n25 = SmoothStep(
   Max=ScalarParameter() {defaultValue=0.2, group=CanvasPrint_Reactivity, parameterName=ReactiveEyes_FadeSmoothness, sortPriority=40}
   Value=OneMinus(None=Abs(None=@n24))
  ) {constMax=0.1}
  NamedRerouteDeclaration(None=@n25) {name=BlueprintTime, nodeColor=<rgba 1 0.146 0.9999 1>}
  Reroute(None=@n25)
 ]
}
